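# Physics asset for the FastCheetah medium male backpack cosmetic, giving collision bodies to the pack and its dyn_egg bone.
# The pack body uses six capsules along the left, right and lower middle, while dyn_egg uses three side-by-side capsules.
# A single constraint joins dyn_egg to pack with swing1 and twist locked, so the egg can rotate only about the swing2 axis.
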PhysicsAsset /BRCosmetics/Accessories/FORT_Backpacks/Backpack_M_MED_FastCheetah/Meshes/M_MED_FastCheetah_Pack_Physics
bodies {
 dyn_egg [
  {
   shapeType=Capsule
   length=6.6337
   radius=3.7625
   shapeName=dyn_egg_01_capsule
   center=<-2.1316 -2.7555 -0.2488>
   rotation=<rot 4.6473 -0.0005 0>
  }
  {shapeType=Capsule, length=11, radius=3.5, shapeName=dyn_egg_02_capsule, center=<0.177 -2.7556 0.4212>, rotation=<rot 6.4386 -0.0002 0>}
  {shapeType=Capsule, length=7, radius=2.5, shapeName=dyn_egg_03_capsule, center=<4.3103 -2.7556 0>, rotation=<rot 3.2692 -0.0001 0>}
 ]
 pack [
  {
   shapeType=Capsule
   length=8.453
   radius=3.9193
   shapeName=pack_mid_01_capsule
   center=<5.8787 -8.8237 -17.6498>
   rotation=<rot -0.6402 -89.0248 -114.0358>
  }
  {
   shapeType=Capsule
   length=10.5962
   radius=3.4026
   shapeName=pack_right_01_capsule
   center=<14.1801 -8.1606 -9.38>
   rotation=<rot -0.6403 -89.0248 -147.516>
  }
  {
   shapeType=Capsule
   length=10.9722
   radius=3.206
   shapeName=pack_left_01_capsule
   center=<-13.1 -9.6687 -10.3101>
   rotation=<rot -0.6431 -89.0248 141.7527>
  }
  {shapeType=Capsule, length=10, radius=3, shapeName=pack_left_02_capsule, center=<-15.3848 -9.7718 0.243>, rotation=<rot -5.8288 0 -0.0002>}
  {shapeType=Capsule, length=9.5, radius=3.5, shapeName=pack_right_02_capsule, center=<14.471 -8.5934 0.5728>, rotation=<rot 11.61 0 -0.0001>}
  {
   shapeType=Capsule
   length=8.4698
   radius=3.9026
   shapeName=pack_mid_02_capsule
   center=<-3.5498 -9.3661 -17.9237>
   rotation=<rot -0.6404 -89.0248 -75.7469>
  }
 ]
}
constraints [{
  bone1Name=dyn_egg
  bone2Name=pack
  swing1LimitDegrees=45
  swing1Motion=Locked
  swing2LimitDegrees=65
  twistLimitDegrees=45
  twistMotion=Locked
 }]
