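# Animation Blueprint for the FastCheetah back bling, adjusting the pack to the wearer's posture and adding secondary motion.
# Event graph derives alphas from owner spine and head bone deltas; the spine alphas blend additive root rotations, while headrot is unused.
# A root offset applies outside skydiving.
# A RigidBody simulation in pack-bone space then feeds four spring-driven AnimDynamics nodes for the egg and yolk bones, relative to dyn_egg.
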AnimBlueprint /BRCosmetics/Accessories/FORT_Backpacks/Backpack_M_MED_FastCheetah/Meshes/M_MED_FastCheetah_Pack_AnimBP
target_skeleton=&/M_MED_FastCheetah_Pack_Skeleton
graphs [
 AnimGraph {nodes [
   n1 = "Transform (Modify) Bone - Bone: root"!(
    ComponentPose=Local To Component(LocalPose=Input Pose())
    bAlphaBoolEnabled:bool=NOT Boolean(A=Get bIsSkyDiving())
    Translation="0,3,0"
    Rotation="0,0,0"
    Scale
   ) -> (Pose) {
    alphaInputType=Bool
    rotationSpace=BCS_BoneSpace
    translationMode=BMM_Additive
    alphaBoolBlend {blendOption=Linear}
    boneToModify {boneName=root}
    translation {y=3}
   }
   n2 = "Transform (Modify) Bone - Bone: root"!(ComponentPose=@n1, Alpha:float=Get rotalpha() -> (rotalpha:float), Translation="0.000,-1.216,0.287", Rotation="-0.000,0.000,13.261", Scale) -> (Pose) {
    rotationMode=BMM_Additive
    rotationSpace=BCS_BoneSpace
    translationMode=BMM_Additive
    translationSpace=BCS_BoneSpace
    alphaBoolBlend {blendOption=Linear}
    boneToModify {boneName=root}
    rotation {roll=13.261}
    translation {y=-1.216, z=0.287}
   }
   n3 = "Transform (Modify) Bone - Bone: root"(ComponentPose=@n2, Alpha=Get yrotalpha() -> (yrotalpha:float), Translation="0,0,0", Rotation="-0.000,0.000,36.139") {
    rotationMode=BMM_Additive
    rotationSpace=BCS_BoneSpace
    alphaBoolBlend {blendOption=Linear}
    boneToModify {boneName=root}
    rotation {roll=36.139}
   }
   n4 = RigidBody(ComponentPose=@n3) {
    bOverrideWorldGravity=true
    overridePhysicsAsset=&/M_MED_FastCheetah_Pack_Physics
    simulationSpace=BaseBoneSpace
    alphaBoolBlend {blendOption=Linear}
    baseBoneRef {boneName=pack}
    componentAppliedLinearAccClamp=<3000 3000 3000>
    componentLinearAccScale {x=0.5, z=0.5}
    componentLinearVelScale {x=1, z=0.3}
    overrideWorldGravity {z=-980}
    simSpaceSettings {
     maxAngularAcceleration=60
     maxAngularVelocity=3
     maxLinearAcceleration=100
     maxLinearVelocity=6
     velocityScaleZ=0
     worldAlpha=1
     externalLinearDragV=<0.1 0.1>
    }
   }
   n5 = Anim Dynamics - Bone: dyn_yolk(ComponentPose=@n4, Alpha=1, ExternalForce="0,0,0") {
    angularSpringConstant=1000
    bAngularSpring=true
    bUseGravityOverride=true
    simulationSpace=RootRelative
    alphaBoolBlend {blendOption=Linear}
    boundBone {boneName=dyn_yolk}
    componentAppliedLinearAccClamp=<3000 3000 3000>
    componentLinearAccScale=<0.45 0.45 0.45>
    componentLinearVelScale=<0.35 0.35 0.35>
    gravityOverride {z=-980}
    physicsBodyDefinitions [{
      boundBone {boneName=dyn_yolk}
      boxExtents=<9 9 9>
      constraintSetup {angularLimitsMax=<0 5 0>, angularLimitsMin=<0 -5 0>, angularTarget=<0 0 0>, linearAxesMax=<0.3 0 0.3>, linearAxesMin=<-0.3 0 -0.3>}
      localJointOffset=<0 0 0>
     }]
    relativeSpaceBone {boneName=dyn_egg}
   }
   n6 = Anim Dynamics - Bone: dyn_yolk_02(ComponentPose=@n5, Alpha=1, ExternalForce="0,0,0") {
    angularSpringConstant=1000
    bAngularSpring=true
    bUseGravityOverride=true
    simulationSpace=RootRelative
    alphaBoolBlend {blendOption=Linear}
    boundBone {boneName=dyn_yolk_02}
    componentAppliedLinearAccClamp=<3000 3000 3000>
    componentLinearAccScale=<0.45 0.45 0.45>
    componentLinearVelScale=<0.35 0.35 0.35>
    gravityOverride {z=-980}
    physicsBodyDefinitions [{
      boundBone {boneName=dyn_yolk_02}
      boxExtents=<7 7 7>
      constraintSetup {angularLimitsMax=<0 5 0>, angularLimitsMin=<0 -5 0>, angularTarget=<0 0 0>, linearAxesMax=<0.4 0 0.4>, linearAxesMin=<-0.4 0 -0.4>}
      localJointOffset=<0.1 0 0>
     }]
    relativeSpaceBone {boneName=dyn_egg}
   }
   n7 = Anim Dynamics - Bone: dyn_right_egg(ComponentPose=@n6, Alpha=1, ExternalForce="0,0,0") {
    angularSpringConstant=1000
    bAngularSpring=true
    bUseGravityOverride=true
    simulationSpace=RootRelative
    alphaBoolBlend {blendOption=Linear}
    boundBone {boneName=dyn_right_egg}
    componentAppliedLinearAccClamp=<3000 3000 3000>
    componentLinearAccScale=<0.45 0.45 0.45>
    componentLinearVelScale=<0.35 0.35 0.35>
    gravityOverride {z=-980}
    physicsBodyDefinitions [{
      boundBone {boneName=dyn_right_egg}
      boxExtents=<8 8 9>
      constraintSetup {angularLimitsMax=<0 6 0>, angularLimitsMin=<0 -6 0>, angularTarget=<0 0 0>, linearAxesMax=<1 0 0>, linearAxesMin=<-1 0 0>}
      localJointOffset=<0 0 0.5>
     }]
    relativeSpaceBone {boneName=dyn_egg}
   }
   n8 = Anim Dynamics - Bone: dyn_left_egg(ComponentPose=@n7, Alpha=1, ExternalForce="0,0,0") {
    angularSpringConstant=2
    bAngularSpring=true
    bUseGravityOverride=true
    simulationSpace=RootRelative
    alphaBoolBlend {blendOption=Linear}
    boundBone {boneName=dyn_left_egg}
    componentLinearAccScale=<0.5 0.5 0.5>
    componentLinearVelScale=<1 1 1>
    gravityOverride {z=-980}
    physicsBodyDefinitions [{
      boundBone {boneName=dyn_left_egg}
      boxExtents=<7 7 7>
      constraintSetup {angularLimitsMax=<2 15 1>, angularLimitsMin=<-2 -20 -1>, angularTarget=<0 0 0>, linearAxesMax=<0.1 0.1 0.1>, linearAxesMin=<-0.1 -0.1 -0.1>}
      localJointOffset=<0 0 -2>
     }]
    relativeSpaceBone {boneName=dyn_egg}
   }
   Output Pose(Result=Component To Local(ComponentPose=@n8)) {layerGroup=None}
  ]}
 EventGraph {nodes [
   Event BlueprintUpdateAnimation()
   n1 = GetOwnerSkeletalMeshComponent()
   > Set rotalpha(rotalpha:float=GetAlphaFromReferenceTransformDelta(SkeletalMeshComponent=@n1, BoneName=spine_05, BaseName=spine_01, InputRange=(X=-20,Y=-36), TwistAxis="1,0,0")) -> (Output_Get:float)
   > Set yrotalpha(yrotalpha:float=GetAlphaFromReferenceTransformDelta(SkeletalMeshComponent=@n1, BoneName=spine_03, BaseName=spine_01, InputRange=(X=10,Y=45), TwistAxis="1,0,0")) -> (Output_Get:float)
   > Set headrot(headrot:float=GetAlphaFromReferenceTransformDelta(SkeletalMeshComponent=@n1, BoneName=head, BaseName=neck_01, InputRange=(X=-47,Y=0), TwistAxis="1,0,0")) -> (Output_Get:float)
   TryGetPawnOwner()
  ]}
]
variables [
 headrot {type=number}
 rotalpha {type=number}
 yrotalpha {type=number}
]
